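# Physics asset for the Scarecrow pickaxe cosmetic, defining physics bodies for its handle and its dangling straw and cloth bones.
# Every body is a single capsule, with one long capsule on Melee_TwoHanded_Handle anchoring the short dyn_ straw and cloth pieces.
# Constraints form short chains or single links off the handle with per-joint swing and twist limits.
# The top cloth root is the exception, with both swings and twist set free.
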PhysicsAsset /BRCosmetics/Weapons/FORT_Melee/Pickaxe_Scarecrow/Meshes/Pickaxe_Scarecrow_Physics
bodies {
 Melee_TwoHanded_Handle [{shapeType=Capsule, length=101.6455, radius=1.5405, center=<0.6519 -0.7347 21.3347>, rotation=<rot -0.1745 -90 0.0358>}]
 dyn_cloth_side_01 [{shapeType=Capsule, length=4.2821, radius=1.2003, rotation=<rot -83.4588 32.7088 -36.5421>}]
 dyn_clothbot_01 [{shapeType=Capsule, length=3.2747, radius=0.75, rotation=<rot 0 90 -90>}]
 dyn_clothbot_02 [{shapeType=Capsule, length=2.8823, radius=1, center=<0.267 0 0>, rotation=<rot 0 90 -90>}]
 dyn_clothbot_03 [{shapeType=Capsule, length=2.3823, radius=1, rotation=<rot 0 90 -90>}]
 dyn_clothtop_01 [{shapeType=Capsule, length=1.5247, radius=1.75, rotation=<rot 0 90 -90>}]
 dyn_clothtop_02 [{shapeType=Capsule, length=1.3823, radius=1.5, center=<0.267 0 0>, rotation=<rot 0 90 -90>}]
 dyn_clothtop_03 [{shapeType=Capsule, length=3.6323, radius=1.5, rotation=<rot 0 90 -90>}]
 dyn_straw_bot [{shapeType=Capsule, length=6.75, radius=9.9258, center=<0 0 -0.5>}]
 dyn_strawa_01 [{shapeType=Capsule, length=3.8303, radius=1.1944, rotation=<rot 0 89.9999 -93.4058>}]
 dyn_strawb_01 [{shapeType=Capsule, length=4.081, radius=1.2752, rotation=<rot 0 90 -90>}]
 dyn_strawb_02 [{shapeType=Capsule, length=2.5906, radius=1.0425, center=<0.292 0 0>, rotation=<rot 0 90 -90>}]
 dyn_strawc_01 [{shapeType=Capsule, length=4.081, radius=1.2752, rotation=<rot 0 90 -90>}]
 dyn_strawc_02 [{shapeType=Capsule, length=2.5906, radius=1.0425, center=<0.292 0 0>, rotation=<rot 0 90 -90>}]
 dyn_strawf_01 [{shapeType=Capsule, length=3.331, radius=1.2752, center=<-0.0597 0 0>, rotation=<rot 0 90 -90>}]
 dyn_strawf_02 [{shapeType=Capsule, length=2.5906, radius=1.0425, rotation=<rot 0 90 -90>}]
 dyn_strawg_01 [{shapeType=Capsule, length=1.081, radius=1.0252, rotation=<rot -90 0 0>}]
 dyn_strawg_02 [{shapeType=Capsule, length=1.505, radius=1.005, rotation=<rot 0 89.7376 -87.7134>}]
}
constraints [
 {
  bone1Name=dyn_cloth_side_01
  bone2Name=Melee_TwoHanded_Handle
  swing1LimitDegrees=6.8571
  swing2LimitDegrees=19.2857
  twistLimitDegrees=3.8571
 }
 {bone1Name=dyn_strawa_01, bone2Name=Melee_TwoHanded_Handle, swing1LimitDegrees=10.2857, swing2LimitDegrees=22.7143, twistLimitDegrees=10.2857}
 {bone1Name=dyn_strawb_01, bone2Name=Melee_TwoHanded_Handle, swing1LimitDegrees=30, swing2LimitDegrees=30, twistLimitDegrees=15}
 {bone1Name=dyn_strawb_02, bone2Name=dyn_strawb_01, swing1LimitDegrees=6.8571, swing2LimitDegrees=19.2857, twistLimitDegrees=3.8571}
 {bone1Name=dyn_strawc_01, bone2Name=Melee_TwoHanded_Handle, swing1LimitDegrees=30, swing2LimitDegrees=30, twistLimitDegrees=15}
 {bone1Name=dyn_strawc_02, bone2Name=dyn_strawc_01, swing1LimitDegrees=6.8571, swing2LimitDegrees=19.2857, twistLimitDegrees=3.8571}
 {bone1Name=dyn_straw_bot, bone2Name=Melee_TwoHanded_Handle, swing1LimitDegrees=45, swing2LimitDegrees=45, twistLimitDegrees=45}
 {bone1Name=dyn_strawf_01, bone2Name=Melee_TwoHanded_Handle, swing1LimitDegrees=6.8571, swing2LimitDegrees=19.2857, twistLimitDegrees=3.8571}
 {bone1Name=dyn_strawf_02, bone2Name=dyn_strawf_01, swing1LimitDegrees=6.8571, swing2LimitDegrees=19.2857, twistLimitDegrees=3.8571}
 {bone1Name=dyn_strawg_01, bone2Name=Melee_TwoHanded_Handle, swing1LimitDegrees=30, swing2LimitDegrees=30, twistLimitDegrees=15}
 {
  bone1Name=dyn_clothbot_01
  bone2Name=Melee_TwoHanded_Handle
  swing1LimitDegrees=10.2857
  swing2LimitDegrees=22.7143
  twistLimitDegrees=10.2857
 }
 {bone1Name=dyn_clothbot_02, bone2Name=dyn_clothbot_01, swing1LimitDegrees=6.8571, swing2LimitDegrees=19.2857, twistLimitDegrees=3.8571}
 {bone1Name=dyn_clothbot_03, bone2Name=dyn_clothbot_02, swing1LimitDegrees=6.8571, swing2LimitDegrees=19.2857, twistLimitDegrees=3.8571}
 {
  bone1Name=dyn_clothtop_01
  bone2Name=Melee_TwoHanded_Handle
  swing1LimitDegrees=10.2857
  swing1Motion=Free
  swing2LimitDegrees=22.7143
  swing2Motion=Free
  twistLimitDegrees=10.2857
  twistMotion=Free
 }
 {bone1Name=dyn_clothtop_02, bone2Name=dyn_clothtop_01, swing1LimitDegrees=6.8571, swing2LimitDegrees=19.2857, twistLimitDegrees=3.8571}
 {bone1Name=dyn_clothtop_03, bone2Name=dyn_clothtop_02, swing1LimitDegrees=6.8571, swing2LimitDegrees=19.2857, twistLimitDegrees=3.8571}
 {bone1Name=dyn_strawg_02, bone2Name=dyn_strawg_01, swing1LimitDegrees=30, swing2LimitDegrees=30, twistLimitDegrees=30}
]
